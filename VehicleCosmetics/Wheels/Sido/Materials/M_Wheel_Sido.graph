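# Masked material for the Sido vehicle wheel, adding animated glowing, see-through effects to the simple wheel parent function.
# Time-driven sine waves along U and V form pulsing bands, tinted or multiplied by a tech texture and vertex color.
# Alpha mask red cuts the shape and green confines effects in color, emissive and opacity, while vertex red restores parent opacity.
# The result passes through shared vehicle damage, then de-rez transition functions.
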Material /VehicleCosmetics/Wheels/Sido/Materials/M_Wheel_Sido
graph {
 properties {
  bAutomaticallySetUsageInEditor=false
  bUseMaterialAttributes=true
  bUsedWithMeshDeformer=true
  bUsedWithSkeletalMesh=true
  blendMode=BLEND_Masked
  pixelDepthOffsetMode=PDOM_Legacy
  previewMesh=/VehicleCosmetics/Wheels/Mild_PSplus/Meshes/SK_Wheel_Mild_PSplus
  usedShadingModels=Default Lit
 }
 nodes [
  n1 = GetMaterialAttributes(None=MaterialFunctionCall() {materialFunction=&MF_WheelParent_Simple}.MaterialAttributes) {attributeGetTypes [#1, #2, #3]}
  n2 = TextureSampleParameter2D() {parameterName=Alpha_Mask_LODs, samplerType=SAMPLERTYPE_LinearColor, texture=/VehicleCosmetics/Wheels/Sido/Textures/T_Wheel_Sido_Alpha_M_LOD}
  n3 = ComponentMask(None=TextureCoordinate()) {r=true}
  n4 = Time()
  n5 = Constant() {r=2}
  n6 = Sine(None=Add(A=@n3, B=Multiply(A=@n4, B=@n5)))
  n7 = Constant() {r=-0.2}
  n8 = Add(A=Multiply(A=@n4, B=Constant() {r=3}), B=@n3)
  n9 = AppendVector(A=Sine(None=Add(A=@n3, B=@n4)), B=AppendVector(A=Add(A=@n6, B=@n7), B=Cosine(None=@n8)))
  n10 = Clamp(None=ConstantBiasScale(None=Desaturation(None=@n9) {luminanceFactors=<rgba 0.3 0.59 0.11 0>}) {bias=0, scale=3})
  n11 = VertexColor()
  n12 = TextureSample() {samplerType=SAMPLERTYPE_LinearColor, texture=/VehicleCosmetics/Wheels/Sido/Textures/Wheel_Sido_Tech_Tex}
  n13 = Multiply(A=Multiply(A=Multiply(A=@n10, B=@n11.A), B=@n12.RGB), B=ScalarParameter() {defaultValue=1, parameterName=Opacity_ef2})
  n14 = ComponentMask(None=TextureCoordinate()) {g=true}
  n15 = Multiply(A=Clamp(None=ConstantBiasScale(None=@n14) {bias=0, scale=1}), B=@n11.A)
  n16 = ScalarParameter() {parameterName=Opasity_Ef3}
  n17 = OneMinus(None=@n2.G)
  n18 = LinearInterpolate(A=Constant(), B=Add(A=@n13, B=Multiply(A=@n15, B=@n16)), Alpha=@n17)
  n19 = LinearInterpolate(A=@n1.Opacity Mask, B=Multiply(A=@n2.R, B=Add(A=@n2.G, B=@n18)), Alpha=OneMinus(None=VertexColor().R))
  n20 = VectorParameter() {parameterName=Color, defaultValue=<rgba 1 1 1 1>}
  n21 = ComponentMask(None=TextureCoordinate()) {g=true}
  n22 = Time()
  n23 = Constant() {r=2}
  n24 = Sine(None=Add(A=@n21, B=Multiply(A=@n22, B=@n23))) {period=1.5}
  n25 = Constant() {r=-0.2}
  n26 = Add(A=Multiply(A=@n22, B=Constant() {r=3}), B=@n21)
  n27 = AppendVector(A=Sine(None=Add(A=@n21, B=@n22)) {period=0.2}, B=AppendVector(A=Add(A=@n24, B=@n25), B=Cosine(None=@n26) {period=0.4}))
  n28 = Clamp(None=ConstantBiasScale(None=Desaturation(None=@n27) {luminanceFactors=<rgba 0.3 0.59 0.11 0>}) {bias=0.8, scale=3})
  n29 = Multiply(A=@n10, B=@n11)
  n30 = Add(
   A=Multiply(A=Multiply(A=@n20.RGB, B=@n28), B=ScalarParameter() {defaultValue=1, parameterName=Opacity_ef1})
   B=Multiply(A=Constant4Vector() {constant=<rgba 1 1 1 1>}.RGB, B=Multiply(A=@n29, B=@n12.RGB))
  )
  n31 = LinearInterpolate(A=@n1.Base Color, B=@n30, Alpha=@n17)
  n32 = LinearInterpolate(A=@n1.Emissive Color, B=@n30, Alpha=OneMinus(None=@n2.G))
  MaterialFunctionCall(In=MaterialFunctionCall(Material=SetMaterialAttributes!(MaterialAttributes=@n1.MaterialAttributes, Opacity Mask=@n19, Base Color=@n31, Emissive Color=@n32) {attributeSetTypes [#1, #2, #3]}) {materialFunction=&MF_BaseVehicle_Damage}.Result) {materialFunction=&MF_VehicleDeRez_Transition}
  Max(A=@n13, B=@n15)
  Multiply(A=@n2.R)
  n33 = Add(A=TextureCoordinate() {uTiling=2, vTiling=2}, B=Constant() {r=-1})
  n34 = ConstantBiasScale(None=OneMinus(None=DotProduct(A=@n33, B=@n33))) {bias=0, scale=1}
  Multiply(A=Multiply(A=Clamp(None=@n34), B=@n11.A), B=@n10)
  OneMinus(None=@n2.R)
 ]
}
